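# UI material for the sale sticker in the frontend store.
# It uses translucent blending and holds a single texture sample of the Store-Sale-Sticker texture with shared world-group clamp sampling.
# The graph shows no other nodes, parameters or connections to material outputs.
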Material /FortniteCoreUI/Frontend/Store/Materials/M-SalesSticker
graph {
 properties {
  bUsedWithStaticMesh=false
  blendMode=BLEND_Translucent
  materialDomain=MD_UI
  pixelDepthOffsetMode=PDOM_Legacy
  usedShadingModels=Default Lit
 }
 nodes [TextureSample() {samplerSource=SSM_Clamp_WorldGroupSettings, texture=/FortniteCoreUI/Frontend/Store/Textures/Store-Sale-Sticker}]
}
